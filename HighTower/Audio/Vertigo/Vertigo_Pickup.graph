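# Vertigo_Pickup is a SoundCue for the pickup sound of Vertigo, likely an item given its loot-collect sound class.
# A single wave player feeds a modulator that holds pitch fixed and applies a constant volume boost, while the cue's own volume multiplier scales the level down.
# It uses a close-range umbrella weapon attenuation, and a concurrency override allows one instance at a time, refusing new plays while one is active.
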SoundCue /HighTower/Audio/Vertigo/Vertigo_Pickup
root=Modulator_0
nodes [
 {id=Modulator_0, children [WavePlayer_0], properties {pitchMax=1, pitchMin=1, volumeMax=1.2, volumeMin=1.2}}
 {id=WavePlayer_0, properties {soundWaveAssetPtr=&/Vertigo_SNG_Pickup_v06}}
]
properties {
 attenuationSettings=/Game/Sounds/Attenuation/Fort_Weapon_Umbrella_Close_Athena_Att
 bOverrideConcurrency=true
 maxDistance=4500
 soundClassObject=&Fort_Loot_Collect
 volumeMultiplier=0.48
 concurrencyOverrides {maxCount=1, resolutionRule=PreventNew}
}
