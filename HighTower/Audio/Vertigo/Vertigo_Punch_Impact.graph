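# Sound cue for the Vertigo punch impact, classed as melee weapon audio.
# A mixer plays thud, pullback, whoosh and a slightly lowered sub hit together as one layered sound.
# A modulator above the mixer applies a fixed volume boost without random pitch or volume variation.
# Concurrency override allows one instance and blocks new ones while it plays, with a borrowed close-range umbrella weapon attenuation for spatial falloff.
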SoundCue /HighTower/Audio/Vertigo/Vertigo_Punch_Impact
root=Modulator_0
nodes [
 {id=Mixer_0, children [WavePlayer_6,WavePlayer_5,WavePlayer_7,WavePlayer_8], properties {inputVolume [1,1,1,0.7]}}
 {id=Modulator_0, children [Mixer_0], properties {pitchMax=1, pitchMin=1, volumeMax=1.25, volumeMin=1.25}}
 {id=WavePlayer_5, properties {soundWaveAssetPtr=&/Vertigo_SNG_Pullback_v04}}
 {id=WavePlayer_6, properties {soundWaveAssetPtr=&/Vertigo_SNG_Impact_Thud}}
 {id=WavePlayer_7, properties {soundWaveAssetPtr=&/Vertigo_SNG_Impact_Whoosh_01}}
 {id=WavePlayer_8, properties {soundWaveAssetPtr=&/Vertigo_Sub_Hit_01}}
]
properties {
 attenuationSettings=/Game/Sounds/Attenuation/Fort_Weapon_Umbrella_Close_Athena_Att
 bOverrideConcurrency=true
 maxDistance=4500
 soundClassObject=&Fort_WeaponMelee
 volumeMultiplier=0.5
 concurrencyOverrides {maxCount=1, resolutionRule=PreventNew}
}
